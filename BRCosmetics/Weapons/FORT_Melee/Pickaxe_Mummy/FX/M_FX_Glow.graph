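# Additive unlit glow material for Niagara sprite and mesh particles, stored with the Mummy pickaxe cosmetic's FX.
# Builds a procedural soft mask from UVs, multiplying smoothstepped per-axis sphere masks, with a parameter blending U toward a one-sided gradient.
# The mask is scaled by particle alpha and an opacity multiplier, then softened by view-angle fade and depth fade.
# A radial mask and a cyan color parameter sit unconnected to that chain.
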Material /BRCosmetics/Weapons/FORT_Melee/Pickaxe_Mummy/FX/M_FX_Glow
graph {
 properties {
  bUsedWithNiagaraMeshParticles=true
  bUsedWithNiagaraSprites=true
  blendMode=BLEND_Additive
  pixelDepthOffsetMode=PDOM_Legacy
  shadingModel=MSM_Unlit
  usedShadingModels=Unlit
 }
 nodes [
  n1 = ScalarParameter() {defaultValue=0.2, parameterName=s}
  n2 = ScalarParameter() {defaultValue=0.75, parameterName=Param_4}
  n3 = MaterialFunctionCall(Float2=TextureCoordinate()) {materialFunction=&BreakOutFloat2Components}
  n4 = ScalarParameter() {defaultValue=0.5, parameterName=Param_1}
  n5 = SphereMask(A=@n3.R, B=@n4, Radius=@n4) {hardnessPercent=0}
  n6 = ScalarParameter() {defaultValue=0.3, parameterName=Param_5}
  n7 = ScalarParameter() {defaultValue=1, parameterName=Param_6}
  n8 = LinearInterpolate(
   A=SmoothStep(Min=@n1, Max=Add(A=@n1, B=@n2), Value=Multiply(A=@n5))
   B=SmoothStep(Min=@n6, Max=Add(A=@n6, B=@n7), Value=@n3.R)
   Alpha=ScalarParameter() {parameterName=UseMaskForSideGlow}
  ) {constAlpha=0}
  n9 = ScalarParameter() {parameterName=Param_2}
  n10 = ScalarParameter() {defaultValue=0.5, parameterName=Param_1}
  n11 = SmoothStep(
   Min=@n9
   Max=Add(A=@n9, B=ScalarParameter() {defaultValue=0.5, parameterName=Param_3})
   Value=SphereMask(A=@n3.G, B=@n10, Radius=@n10) {hardnessPercent=0}
  )
  n12 = Multiply(A=Power(Base=Multiply(A=@n8, B=@n11)), B=ParticleColor().A)
  n13 = ScalarParameter() {defaultValue=1, parameterName=Opacity Mult}
  n14 = ScalarParameter() {defaultValue=2, parameterName=Angle Fade}
  DepthFade(
   Opacity=Multiply(
    A=Multiply(A=@n12, B=@n13)
    B=MaterialFunctionCall("FadeSensitivity (optional)"=@n14) {materialFunction=&FadeBasedOnViewAngle}.Result
   )
   FadeDistance=ScalarParameter() {parameterName=Fade Distance}
  )
  n15 = ScalarParameter() {defaultValue=0.5}
  Power(Base=SphereMask(A=TextureCoordinate(), B=@n15, Radius=@n15) {hardnessPercent=0}) {constExponent=1}
  VectorParameter() {defaultValue=<rgba 0.159 0.807 1 1>}
 ]
}
