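# Niagara anim-trail effect for the Mummy pickaxe, drawing layered swing ribbons from bone or socket pairs on the weapon.
# Seven short-lived ribbon emitters (Air, Thin, Underlay, two Base, Strip, Blue_01) share the FN AnimTrail spawn, initialize and update modules, varying in socket pair, width and color.
# Curves fade each layer and taper most over lifetime.
# FNTimeOfDayScalar adjusts every layer's opacity and, except on Air, blends RGB tint by time of day.
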NiagaraSystem /BRCosmetics/Weapons/FORT_Melee/Pickaxe_Mummy/FX/NS_PickaxeMummy_AnimTrail
dependencies {
 usedDataInterfaces [&NiagaraDataInterfaceColorCurve, &NiagaraDataInterfaceCurve]
 usedDynamicInputs []
 usedModules [&SystemState]
}
emitters {
 Air {
  emitterUpdateScript [
   &EmitterState()
   &FN_SpawnAnimTrail_v2(Spawn Interval=0)
  ]
  particleSpawnScript [
   /Game/Effects/Niagara/Modules/AnimTrail/FN_InitializeAnimTrail/FN_InitializeAnimTrailBoneIndecies(Second Bone or Socket Index=1)
   /Game/Effects/Niagara/Modules/AnimTrail/FN_InitializeAnimTrail/FN_InitializeAnimTrail(AnimTrailWidth=1, Color=<rgba 1 1 1 0.2>, Lifetime=0.225)
  ]
  particleUpdateScript [
   &SolveForcesAndVelocity(Clamp Velocity=false, Limit Acceleration=false)
   &FN_UpdateAnimTrail_v2()
   &Color(Color=&ColorFromCurve(ColorCurve=DataInterface() {propertyValues {
       AlphaCurve {keys [<at 0 0.25 RCIM_Cubic>,<at 1 0 RCIM_Cubic>]}
       BlueCurve {keys [<at 0.2607 0.35 RCIM_Cubic>,<at 0.7953 0.835>]}
       GreenCurve {keys [<at 0.2607 0.6208 RCIM_Cubic>,<at 0.7953 0.9038>]}
       RedCurve {keys [<at 0.2607 1 RCIM_Cubic>,<at 0.7953 1>]}
      }}))
   &ScaleColor(
    Color Value To Scale=$StackContext.Color
    Scale Alpha=&FloatFromCurve(
     FloatCurve=DataInterface() {propertyValues {Curve {keys [<at 0 1 RCIM_Cubic>,<at 0.6 0>]}}}
     Scale Curve=&FNTimeOfDayScalar(Night Scale=0.2, Morning Scale=0.6, Evening Scale=0.6)
    )
    Scale RGB=&VectorFromFloat(Value=2)
    ScaleA=true
    ScaleRGB=true
   )
  ]
  rendererClasses [&NiagaraRibbonRendererProperties]
  renderers [{rendererClass=&NiagaraRibbonRendererProperties}]
 }
 Base_01 {
  emitterUpdateScript [
   &EmitterState()
   &FN_SpawnAnimTrail_v2()
  ]
  particleSpawnScript [
   &FN_InitializeAnimTrailBoneIndecies(First Bone or Socket Index=1, Second Bone or Socket Index=2)
   &FN_InitializeAnimTrail(AnimTrailWidth=1, Color=<rgba 1 0.6115 0.334 1>, Lifetime=0.18)
   &SystemLocation(MaskSpawnGroup=false)
  ]
  particleUpdateScript [
   &ScaleRibbonWidth(Ribbon Width Scale=&FloatFromCurve(FloatCurve=DataInterface() {propertyValues {Curve {keys [
         {arriveTangent=-1.2579, interpMode=RCIM_Cubic, time=0, value=1}
         {arriveTangent=-0.9664, interpMode=RCIM_Cubic, leaveTangent=-0.9664, tangentMode=RCTM_User, time=1, value=0}
        ]}}}))
   &ScaleColor(
    Scale Alpha=&FloatFromCurve(
     FloatCurve=DataInterface() {propertyValues {Curve {keys [
         {interpMode=RCIM_Cubic, leaveTangent=-2.3646, tangentMode=RCTM_Break, time=0, value=1}
         <at 0.6689 0.0233>
         {arriveTangent=-0.2139, interpMode=RCIM_Cubic, tangentMode=RCTM_Break, time=1, value=0}
        ]}}}
     Scale Curve=0.9
    )
    ScaleA=true
    ScaleRGB=true
   ) {module=ScaleColor002}
   &ScaleColor(
    Color Value To Scale=$Particles.Color
    Scale Alpha=&FNTimeOfDayScalar(Night Scale=0.3, Morning Scale=0.7, Day Scale=0.85, Evening Scale=0.7)
    Scale RGB=/Niagara/DynamicInputs/Math/LerpMultipleVectors(
     Alpha:NiagaraFloat=&FNTimeOfDayScalar(Night Scale=0, Morning Scale=0.333, Day Scale=0.666, Evening Scale=1)
     V 1:Vector3f=&MakeVectorFromLinearColorRGB(Color=<rgba 0.7 0.815 1 1>)
     V 2:Vector3f=&MakeVectorFromLinearColorRGB(Color=<rgba 1 0.8548 0.85 1>)
     V 3:Vector3f=&MakeVectorFromLinearColorRGB(Color=<rgba 1 1 1 1>)
     V 4:Vector3f=&MakeVectorFromLinearColorRGB(Color=<rgba 1 0.8548 0.85 1>)
     V 5:Vector3f=?
     Vector Count:ENiagara_MultipleLerpCount=::NewEnumerator1
    )
    ScaleA=true
    ScaleRGB=true
   )
   &FN_UpdateAnimTrail_v2()
  ]
  rendererClasses [&NiagaraRibbonRendererProperties]
  renderers [{rendererClass=&NiagaraRibbonRendererProperties}]
 }
 Base_02 {
  emitterUpdateScript [
   &EmitterState()
   &FN_SpawnAnimTrail_v2()
  ]
  particleSpawnScript [
   &FN_InitializeAnimTrailBoneIndecies(First Bone or Socket Index=3, Second Bone or Socket Index=2)
   &FN_InitializeAnimTrail(AnimTrailWidth=1, Color=<rgba 1 0.6115 0.334 1>, Lifetime=0.18)
   &SystemLocation(MaskSpawnGroup=false)
  ]
  particleUpdateScript [
   &ScaleRibbonWidth(Ribbon Width Scale=&FloatFromCurve(FloatCurve=DataInterface() {propertyValues {Curve {keys [
         {arriveTangent=-1.2579, interpMode=RCIM_Cubic, time=0, value=1}
         {arriveTangent=-0.9664, interpMode=RCIM_Cubic, leaveTangent=-0.9664, tangentMode=RCTM_User, time=1, value=0}
        ]}}}))
   &ScaleColor(
    Scale Alpha=&FloatFromCurve(
     FloatCurve=DataInterface() {propertyValues {Curve {keys [
         {interpMode=RCIM_Cubic, leaveTangent=-2.3646, tangentMode=RCTM_Break, time=0, value=1}
         <at 0.6689 0.0233>
         {arriveTangent=-0.2139, interpMode=RCIM_Cubic, tangentMode=RCTM_Break, time=1, value=0}
        ]}}}
     Scale Curve=0.9
    )
    ScaleA=true
    ScaleRGB=true
   ) {module=ScaleColor002}
   &ScaleColor(
    Color Value To Scale=$Particles.Color
    Scale Alpha=&FNTimeOfDayScalar(Night Scale=0.3, Morning Scale=0.7, Day Scale=0.85, Evening Scale=0.7)
    Scale RGB=&LerpMultipleVectors(
     Alpha:NiagaraFloat=&FNTimeOfDayScalar(Night Scale=0, Morning Scale=0.333, Day Scale=0.666, Evening Scale=1)
     V 1:Vector3f=&MakeVectorFromLinearColorRGB(Color=<rgba 0.7 0.815 1 1>)
     V 2:Vector3f=&MakeVectorFromLinearColorRGB(Color=<rgba 1 0.8548 0.85 1>)
     V 3:Vector3f=&MakeVectorFromLinearColorRGB(Color=<rgba 1 1 1 1>)
     V 4:Vector3f=&MakeVectorFromLinearColorRGB(Color=<rgba 1 0.8548 0.85 1>)
     V 5:Vector3f=?
     Vector Count:ENiagara_MultipleLerpCount=::NewEnumerator1
    )
    ScaleA=true
    ScaleRGB=true
   )
   &FN_UpdateAnimTrail_v2()
  ]
  rendererClasses [&NiagaraRibbonRendererProperties]
  renderers [{rendererClass=&NiagaraRibbonRendererProperties}]
 }
 Blue_01 {
  emitterUpdateScript [
   &EmitterState()
   &FN_SpawnAnimTrail_v2(AnimTrailWidthMode=::FromFirst)
  ]
  particleSpawnScript [
   &FN_InitializeAnimTrailBoneIndecies(First Bone or Socket Index=5, Second Bone or Socket Index=2)
   &FN_InitializeAnimTrail(AnimTrailWidth=0.1, Color=<rgba 0.186 0.8907 1 0.05>, Lifetime=0.12)
   &SystemLocation(MaskSpawnGroup=false)
  ]
  particleUpdateScript [
   &ScaleRibbonWidth(Ribbon Width Scale=&FloatFromCurve(FloatCurve=DataInterface() {propertyValues {Curve {keys [
         {arriveTangent=-1.2579, interpMode=RCIM_Cubic, time=0, value=1}
         {arriveTangent=-0.9664, interpMode=RCIM_Cubic, leaveTangent=-0.9664, tangentMode=RCTM_User, time=1, value=0}
        ]}}}))
   &ScaleColor(
    Scale Alpha=&FloatFromCurve(
     FloatCurve=DataInterface() {propertyValues {Curve {keys [
         {arriveTangent=-2.003, interpMode=RCIM_Cubic, leaveTangent=-2.003, tangentMode=RCTM_User, time=0.5, value=0.95}
         {interpMode=RCIM_Cubic, tangentMode=RCTM_User, time=1, value=0}
        ]}}}
     Scale Curve=&Multiply_Float(A=0.5, B=&FNTimeOfDayScalar(Night Scale=0.9, Morning Scale=0.9, Evening Scale=0.9))
    )
    Scale RGB=&Multiply_Vector(
     A=&MakeVectorFromLinearColorRGB(Color=&ColorFromCurve(ColorCurve=DataInterface() {propertyValues {
         AlphaCurve {keys [<at 0 1 RCIM_Cubic>,<at 1 0 RCIM_Cubic>]}
         BlueCurve {keys [<at 0 1 RCIM_Cubic>,<at 0.3065 1 RCIM_Cubic>]}
         GreenCurve {keys [<at 0 1 RCIM_Cubic>,<at 0.3065 1 RCIM_Cubic>]}
         RedCurve {keys [<at 0 1 RCIM_Cubic>,<at 0.3065 1 RCIM_Cubic>]}
        }}))
     B=&LerpMultipleVectors(
      Alpha:NiagaraFloat=&FNTimeOfDayScalar(Night Scale=0, Morning Scale=0.333, Day Scale=0.666, Evening Scale=1)
      V 1:Vector3f=&MakeVectorFromLinearColorRGB(Color=<rgba 0.513 0.6997 1 1>)
      V 2:Vector3f=&MakeVectorFromLinearColorRGB(Color=<rgba 1 0.9032 0.9 1>)
      V 3:Vector3f=&MakeVectorFromLinearColorRGB(Color=<rgba 1 1 1 1>)
      V 4:Vector3f=&MakeVectorFromLinearColorRGB(Color=<rgba 1 0.9032 0.9 1>)
      V 5:Vector3f=?
      Vector Count:ENiagara_MultipleLerpCount=::NewEnumerator1
     )
    )
    ScaleA=true
    ScaleRGB=true
   ) {module=ScaleColor001}
   &FN_UpdateAnimTrail_v2()
  ]
  rendererClasses [&NiagaraRibbonRendererProperties]
  renderers [{rendererClass=&NiagaraRibbonRendererProperties}]
 }
 Strip {
  emitterUpdateScript [
   &EmitterState()
   &FN_SpawnAnimTrail_v2(AnimTrailWidthMode=::FromSecond)
  ]
  particleSpawnScript [
   &FN_InitializeAnimTrailBoneIndecies(First Bone or Socket Index=4, Second Bone or Socket Index=1)
   &FN_InitializeAnimTrail(AnimTrailWidth=1, Color=<rgba 0.666 0.4049 0.1485 0.888>, Lifetime=0.2)
   &SystemLocation(MaskSpawnGroup=false)
  ]
  particleUpdateScript [
   &ScaleRibbonWidth(Ribbon Width Scale=&FloatFromCurve(FloatCurve=DataInterface() {propertyValues {Curve {keys [
         {arriveTangent=-1.2579, interpMode=RCIM_Cubic, time=0, value=1}
         {arriveTangent=-0.9664, interpMode=RCIM_Cubic, leaveTangent=-0.9664, tangentMode=RCTM_User, time=1, value=0}
        ]}}}))
   &ScaleColor(
    Scale Alpha=&FloatFromCurve(
     FloatCurve=DataInterface() {propertyValues {Curve {keys [
         {arriveTangent=-2.003, interpMode=RCIM_Cubic, leaveTangent=-2.003, tangentMode=RCTM_User, time=0, value=0.95}
         {interpMode=RCIM_Cubic, tangentMode=RCTM_User, time=0.5, value=0}
        ]}}}
     Scale Curve=&Multiply_Float(A=0.5, B=&FNTimeOfDayScalar(Night Scale=0.3, Morning Scale=0.25, Evening Scale=0.25))
    )
    Scale RGB=&Multiply_Vector(
     A=&MakeVectorFromLinearColorRGB(Color=&ColorFromCurve(ColorCurve=DataInterface() {propertyValues {
         AlphaCurve {keys [<at 0 1 RCIM_Cubic>,<at 1 0 RCIM_Cubic>]}
         BlueCurve {keys [<at 0 1.8 RCIM_Cubic>,<at 0.3065 1 RCIM_Cubic>]}
         GreenCurve {keys [<at 0 1.8 RCIM_Cubic>,<at 0.3065 1 RCIM_Cubic>]}
         RedCurve {keys [<at 0 1.8 RCIM_Cubic>,<at 0.3065 1 RCIM_Cubic>]}
        }}))
     B=&LerpMultipleVectors(
      Alpha:NiagaraFloat=&FNTimeOfDayScalar(Night Scale=0, Morning Scale=0.333, Day Scale=0.666, Evening Scale=1)
      V 1:Vector3f=&MakeVectorFromLinearColorRGB(Color=<rgba 0.513 0.6997 1 1>)
      V 2:Vector3f=&MakeVectorFromLinearColorRGB(Color=<rgba 1 0.9032 0.9 1>)
      V 3:Vector3f=&MakeVectorFromLinearColorRGB(Color=<rgba 1 1 1 1>)
      V 4:Vector3f=&MakeVectorFromLinearColorRGB(Color=<rgba 1 0.9032 0.9 1>)
      V 5:Vector3f=?
      Vector Count:ENiagara_MultipleLerpCount=::NewEnumerator1
     )
    )
    ScaleA=true
    ScaleRGB=true
   ) {module=ScaleColor001}
   &FN_UpdateAnimTrail_v2()
  ]
  rendererClasses [&NiagaraRibbonRendererProperties]
  renderers [{rendererClass=&NiagaraRibbonRendererProperties}]
 }
 Thin {
  emitterUpdateScript [
   &EmitterState()
   &FN_SpawnAnimTrail_v2(AnimTrailWidthMode=::FromSecond)
  ]
  particleSpawnScript [
   &FN_InitializeAnimTrailBoneIndecies(Second Bone or Socket Index=1)
   &FN_InitializeAnimTrail(AnimTrailWidth=0.05, Color=<rgba 1.25 0.769 0.4179 1>, Lifetime=0.2)
   &SystemLocation(MaskSpawnGroup=false)
  ]
  particleUpdateScript [
   &ScaleRibbonWidth(Ribbon Width Scale=&FloatFromCurve(FloatCurve=DataInterface() {propertyValues {Curve {keys [
         {arriveTangent=-1.2579, interpMode=RCIM_Cubic, time=0, value=1}
         {arriveTangent=-0.9664, interpMode=RCIM_Cubic, leaveTangent=-0.9664, tangentMode=RCTM_User, time=1, value=0}
        ]}}}))
   &ScaleColor(
    Scale Alpha=&FloatFromCurve(
     FloatCurve=DataInterface() {propertyValues {Curve {keys [
         {interpMode=RCIM_Cubic, leaveTangent=-2.9794, tangentMode=RCTM_Break, time=0, value=1}
         <at 0.4 0.0233>
         {arriveTangent=-0.0194, interpMode=RCIM_Cubic, tangentMode=RCTM_Break, time=0.8, value=0}
        ]}}}
     Scale Curve=&Multiply_Float(
      A=0.5
      B=&FNTimeOfDayScalar(Night Scale=0.3, Morning Scale=0.25, Day Scale=0.9, Evening Scale=0.25)
     )
    )
    Scale RGB=&Multiply_Vector(
     A=&MakeVectorFromLinearColorRGB(Color=&ColorFromCurve(ColorCurve=DataInterface() {propertyValues {
         AlphaCurve {keys [<at 0 1 RCIM_Cubic>,<at 1 0 RCIM_Cubic>]}
         BlueCurve {keys [<at 0 1.8 RCIM_Cubic>,<at 0.3065 1 RCIM_Cubic>]}
         GreenCurve {keys [<at 0 1.8 RCIM_Cubic>,<at 0.3065 1 RCIM_Cubic>]}
         RedCurve {keys [<at 0 1.8 RCIM_Cubic>,<at 0.3065 1 RCIM_Cubic>]}
        }}))
     B=&LerpMultipleVectors(
      Alpha:NiagaraFloat=&FNTimeOfDayScalar(Night Scale=0, Morning Scale=0.333, Day Scale=0.666, Evening Scale=1)
      V 1:Vector3f=&MakeVectorFromLinearColorRGB(Color=<rgba 0.513 0.6997 1 1>)
      V 2:Vector3f=&MakeVectorFromLinearColorRGB(Color=<rgba 1 0.9032 0.9 1>)
      V 3:Vector3f=&MakeVectorFromLinearColorRGB(Color=<rgba 1 1 1 1>)
      V 4:Vector3f=&MakeVectorFromLinearColorRGB(Color=<rgba 1 0.9032 0.9 1>)
      V 5:Vector3f=?
      Vector Count:ENiagara_MultipleLerpCount=::NewEnumerator1
     )
    )
    ScaleA=true
    ScaleRGB=true
   ) {module=ScaleColor001}
   &FN_UpdateAnimTrail_v2()
  ]
  rendererClasses [&NiagaraRibbonRendererProperties]
  renderers [{rendererClass=&NiagaraRibbonRendererProperties}]
 }
 Underlay {
  emitterUpdateScript [
   &EmitterState()
   &FN_SpawnAnimTrail_v2()
  ]
  particleSpawnScript [
   &FN_InitializeAnimTrailBoneIndecies(Second Bone or Socket Index=1)
   &FN_InitializeAnimTrail(AnimTrailWidth=1, Color=<rgba 0.025 0.0162 0.0021 1.333>, Lifetime=0.22)
   &SystemLocation(MaskSpawnGroup=false)
  ]
  particleUpdateScript [
   &FN_UpdateAnimTrail_v2()
   &ScaleColor(
    Scale Alpha=&Multiply_Float(
     A=&FloatFromCurve(
      FloatCurve=DataInterface() {propertyValues {Curve {keys [{tangentMode=RCTM_User, time=0, value=1}, {tangentMode=RCTM_User, time=1, value=0}]}}}
      Scale Curve=0.7
     )
     B=&FNTimeOfDayScalar(Night Scale=0.5, Morning Scale=0.8, Evening Scale=0.8)
    )
    Scale RGB=&LerpMultipleVectors(
     Alpha:NiagaraFloat=&FNTimeOfDayScalar(Night Scale=0, Morning Scale=0.333, Day Scale=0.666, Evening Scale=1)
     V 1:Vector3f=&MakeVectorFromLinearColorRGB(Color=<rgba 0.513 0.6997 1 1>)
     V 2:Vector3f=&MakeVectorFromLinearColorRGB(Color=<rgba 1 0.8548 0.85 1>)
     V 3:Vector3f=&MakeVectorFromLinearColorRGB(Color=<rgba 1 1 1 1>)
     V 4:Vector3f=&MakeVectorFromLinearColorRGB(Color=<rgba 1 0.8548 0.85 1>)
     V 5:Vector3f=?
     Vector Count:ENiagara_MultipleLerpCount=::NewEnumerator1
    )
    ScaleA=true
    ScaleRGB=true
   ) {module=ScaleColor001}
  ]
  rendererClasses [&NiagaraRibbonRendererProperties]
  renderers [{rendererClass=&NiagaraRibbonRendererProperties}]
 }
}
summary {
 systemName=NS_PickaxeMummy_AnimTrail
 emitters [
  {emitterName=Air, rendererClasses [&NiagaraRibbonRendererProperties]}
  {emitterName=Thin, rendererClasses [&NiagaraRibbonRendererProperties]}
  {emitterName=Underlay, rendererClasses [&NiagaraRibbonRendererProperties]}
  {emitterName=Base_01, rendererClasses [&NiagaraRibbonRendererProperties]}
  {emitterName=Base_02, rendererClasses [&NiagaraRibbonRendererProperties]}
  {emitterName=Strip, rendererClasses [&NiagaraRibbonRendererProperties]}
  {emitterName=Blue_01, rendererClasses [&NiagaraRibbonRendererProperties]}
 ]
 userVariables [
  User.bIsFrontEndPreview {type=NiagaraBool}
  User.RibbonIDOffset {type=NiagaraInt32}
  User.SkeletalMesh {
   type=NiagaraDataInterfaceSkeletalMesh
   defaultValue {
    struct=NiagaraExt_VariableValue_DataInterface
    value {dataInterface=&self:NiagaraDataInterfaceSkeletalMesh_0}
   }
  }
  User.Transition {type=NiagaraFloat, defaultValue=0, description=""}
 ]
}
